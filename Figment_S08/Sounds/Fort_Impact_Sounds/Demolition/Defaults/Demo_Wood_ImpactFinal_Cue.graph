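# Sound cue for the final wood impact in demolition, playing a layered hit that changes with listener distance.
# A Mixer blends three randomized layers, each with its own close, medium or far crossfade attenuation.
# Close uses briefly delayed, pitched-up stereo impacts, medium uses modulated final-impact waves, and far uses short small impacts.
# Output is routed to the Fort impacts sound class, with a shared four-voice concurrency set.
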SoundCue /Figment_S08/Sounds/Fort_Impact_Sounds/Demolition/Defaults/Demo_Wood_ImpactFinal_Cue
root=Mixer_1
nodes [
 {id=Attenuation_2, children [Random_0], properties {attenuationSettings=/Figment_S08/Sounds/Attenuation/Fort_Demolition_Far_Crossfade_Att}}
 {
  id=Attenuation_5
  children [Modulator_1]
  properties {attenuationSettings=/Figment_S08/Sounds/Attenuation/Fort_Demolition_Medium_Crossfade_Att}
 }
 {id=Attenuation_6, children [Delay_2], properties {attenuationSettings=/Figment_S08/Sounds/Attenuation/Fort_Demolition_Close_Crossfade_Att}}
 {id=Delay_2, children [Modulator_3], properties {delayMax=0.1, delayMin=0.05}}
 {id=Mixer_1, children [Attenuation_6, Attenuation_5, Attenuation_2], properties {inputVolume [1,1,1]}}
 {id=Modulator_1, children [Random_1]}
 {id=Modulator_3, children [Random_9], properties {pitchMax=1.3, pitchMin=1.1}}
 {id=Random_0, children [WavePlayer_0,WavePlayer_1,WavePlayer_2,WavePlayer_3], properties {weights [1,1,1,1]}}
 {id=Random_1, children [WavePlayer_89, WavePlayer_105, WavePlayer_106], properties {weights [1,1,1]}}
 {id=Random_9, children [WavePlayer_88, WavePlayer_94, WavePlayer_91, WavePlayer_93], properties {weights [1,1,1,1]}}
 {id=WavePlayer_0, properties {soundWaveAssetPtr=/Figment_S08/Sounds/Fort_Impact_Sounds/Demolition/Wood_ImpactShort_Small08}}
 {id=WavePlayer_1, properties {soundWaveAssetPtr=/Figment_S08/Sounds/Fort_Impact_Sounds/Demolition/Wood_ImpactShort_Small07}}
 {id=WavePlayer_105, properties {soundWaveAssetPtr=/Figment_S08/Sounds/Fort_Impact_Sounds/Demolition/Fort_Impact_WoodFinal_02}}
 {id=WavePlayer_106, properties {soundWaveAssetPtr=/Figment_S08/Sounds/Fort_Impact_Sounds/Demolition/Fort_Impact_WoodFinal_03}}
 {id=WavePlayer_2, properties {soundWaveAssetPtr=/Figment_S08/Sounds/Fort_Impact_Sounds/Demolition/Wood_ImpactShort_Small09}}
 {id=WavePlayer_3, properties {soundWaveAssetPtr=/Figment_S08/Sounds/Fort_Impact_Sounds/Demolition/Wood_ImpactShort_Small10}}
 {id=WavePlayer_88, properties {soundWaveAssetPtr=/Figment_S08/Sounds/Fort_Impact_Sounds/Demolition/wood_impact_stereo_01}}
 {id=WavePlayer_89, properties {soundWaveAssetPtr=/Figment_S08/Sounds/Fort_Impact_Sounds/Demolition/Fort_Impact_WoodFinal_01}}
 {id=WavePlayer_91, properties {soundWaveAssetPtr=/Figment_S08/Sounds/Fort_Impact_Sounds/Demolition/wood_impact_stereo_07}}
 {id=WavePlayer_93, properties {soundWaveAssetPtr=/Figment_S08/Sounds/Fort_Impact_Sounds/Demolition/wood_impact_stereo_08}}
 {id=WavePlayer_94, properties {soundWaveAssetPtr=/Figment_S08/Sounds/Fort_Impact_Sounds/Demolition/wood_impact_stereo_04}}
]
properties {
 maxDistance=4500
 soundClassObject=&Fort_Impacts_PSM
 volumeMultiplier=0.5
 concurrencyOverrides {resolutionRule=PreventNew}
 concurrencySet [/Game/Sounds/Concurrency_Rules/FourMaximum_StopFurthestOldest_VolumeScale]
}
